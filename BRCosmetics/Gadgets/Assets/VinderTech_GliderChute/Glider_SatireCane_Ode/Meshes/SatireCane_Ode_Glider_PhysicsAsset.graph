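# Physics asset for the SatireCane_Ode glider cosmetic, a creature-shaped skeletal mesh with a body, jaw, four legs and a tail.
# Each simulated bone gets a single capsule: the body, the jaw, three segments per paw chain and two tail bones.
# Constraints chain each limb and the tail back to Body_Offset, with the jaw attached directly, all sharing uniform swing and twist limits.
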PhysicsAsset /BRCosmetics/Gadgets/Assets/VinderTech_GliderChute/Glider_SatireCane_Ode/Meshes/SatireCane_Ode_Glider_PhysicsAsset
bodies {
 Body_Offset [{shapeType=Capsule, length=146.5264, radius=65.5035, center=<0.0321 -57.6095 -31.7004>, rotation=<rot 0 0.0151 -82.6551>}]
 jaw_01 [{shapeType=Capsule, length=21.0307, radius=18.85, center=<13.1132 1.84 0.9484>, rotation=<rot 0 88.0968 -83.7777>}]
 paw_back_01_l [{shapeType=Capsule, length=60.5253, radius=46.2907, center=<20.3382 8.2825 1.2973>, rotation=<rot 0 86.2762 -65.4781>}]
 paw_back_01_r [{shapeType=Capsule, length=60.5236, radius=46.347, center=<-20.3078 -8.2551 -1.2952>, rotation=<rot 0 86.2248 -65.3928>}]
 paw_back_02_l [{shapeType=Capsule, length=44.714, radius=20.7847, center=<31.4637 2.1934 -0.0431>, rotation=<rot 0 84.5996 -89.9208>}]
 paw_back_02_r [{shapeType=Capsule, length=44.698, radius=20.7866, center=<-31.4665 -2.172 -0.0148>, rotation=<rot 0 -95.46 -89.9728>}]
 paw_back_03_l [{shapeType=Capsule, length=38.96, radius=30.0127, center=<27.6785 -12.8049 -0.0095>, rotation=<rot 0 -147.8214 -89.9741>}]
 paw_back_03_r [{shapeType=Capsule, length=38.9635, radius=30.0131, center=<-27.6811 12.8004 0.0127>, rotation=<rot 0 -147.8118 -89.944>}]
 paw_f_01_l [{shapeType=Capsule, length=67.1535, radius=31.0894, center=<-2.4239 -2.5779 -0.8535>, rotation=<rot 0 85.665 -83.8084>}]
 paw_f_01_r [{shapeType=Capsule, length=67.1136, radius=31.3576, center=<2.4846 2.7674 1.003>, rotation=<rot 0 85.1467 -83.6411>}]
 paw_f_02_l [{shapeType=Capsule, length=50.5055, radius=23.7836, center=<40.2291 -1.3722 0.1287>, rotation=<rot 0 -87.14 -89.6039>}]
 paw_f_02_r [{shapeType=Capsule, length=50.4818, radius=23.7861, center=<-40.2837 1.3612 0.0778>, rotation=<rot 0 -87.1276 -89.8418>}]
 paw_f_03_l [{shapeType=Capsule, length=45.9547, radius=30.5777, center=<34.1513 -17.7702 0.1116>, rotation=<rot 0 -140.6779 -89.7283>}]
 paw_f_03_r [{shapeType=Capsule, length=45.9159, radius=30.5597, center=<-34.1681 17.772 0.0937>, rotation=<rot 0 -140.6473 -89.8485>}]
 tail_01 [{shapeType=Capsule, length=18.822, radius=4.707, center=<7.7589 0 -0.44>, rotation=<rot 0 -90 -89.1834>}]
 tail_03 [{shapeType=Capsule, length=81.2677, radius=17.0341, center=<68.8662 -0.1905 -6.7136>, rotation=<rot 0 -90.1074 -86.1118>}]
}
constraints [
 {bone1Name=paw_back_01_l, bone2Name=Body_Offset, swing1LimitDegrees=45, swing2LimitDegrees=45, twistLimitDegrees=45}
 {bone1Name=paw_back_02_l, bone2Name=paw_back_01_l, swing1LimitDegrees=45, swing2LimitDegrees=45, twistLimitDegrees=45}
 {bone1Name=paw_back_03_l, bone2Name=paw_back_02_l, swing1LimitDegrees=45, swing2LimitDegrees=45, twistLimitDegrees=45}
 {bone1Name=paw_back_01_r, bone2Name=Body_Offset, swing1LimitDegrees=45, swing2LimitDegrees=45, twistLimitDegrees=45}
 {bone1Name=paw_back_02_r, bone2Name=paw_back_01_r, swing1LimitDegrees=45, swing2LimitDegrees=45, twistLimitDegrees=45}
 {bone1Name=paw_back_03_r, bone2Name=paw_back_02_r, swing1LimitDegrees=45, swing2LimitDegrees=45, twistLimitDegrees=45}
 {bone1Name=paw_f_01_r, bone2Name=Body_Offset, swing1LimitDegrees=45, swing2LimitDegrees=45, twistLimitDegrees=45}
 {bone1Name=paw_f_02_r, bone2Name=paw_f_01_r, swing1LimitDegrees=45, swing2LimitDegrees=45, twistLimitDegrees=45}
 {bone1Name=paw_f_03_r, bone2Name=paw_f_02_r, swing1LimitDegrees=45, swing2LimitDegrees=45, twistLimitDegrees=45}
 {bone1Name=paw_f_01_l, bone2Name=Body_Offset, swing1LimitDegrees=45, swing2LimitDegrees=45, twistLimitDegrees=45}
 {bone1Name=paw_f_02_l, bone2Name=paw_f_01_l, swing1LimitDegrees=45, swing2LimitDegrees=45, twistLimitDegrees=45}
 {bone1Name=paw_f_03_l, bone2Name=paw_f_02_l, swing1LimitDegrees=45, swing2LimitDegrees=45, twistLimitDegrees=45}
 {bone1Name=tail_01, bone2Name=Body_Offset, swing1LimitDegrees=45, swing2LimitDegrees=45, twistLimitDegrees=45}
 {bone1Name=tail_03, bone2Name=tail_01, swing1LimitDegrees=45, swing2LimitDegrees=45, twistLimitDegrees=45}
 {bone1Name=jaw_01, bone2Name=Body_Offset, swing1LimitDegrees=45, swing2LimitDegrees=45, twistLimitDegrees=45}
]
